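; Ballistic shield block volume that attaches to the wielder's shield mesh and absorbs hits with a collision box, scaled separately on the server.
; Server damage resets building health, fires cues and events, and, in eligible shield states, adds tag-multiplied damage to a stagger count seeded from the weapon.
; Reaching the block threshold enters stagger, whose OnRep sends the stagger event and disables collision.
; Timers decay the count after a delay without damage.
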(event Custom|TestBallisticShieldStagger
  ;[1 0 0.05961] DEBUG
  (if (not (Variables|Default|GetInStaggerState?))
    (Variables|Default|SetInStaggerState? true)))

(event EventBeginPlay
  (bind _owningweapon (Variables|Default|GetOwningWeapon))
  (bind _owningweapon_2 (Variables|Default|GetOwningWeapon))
  ;[0.03429 0.3069 0.7899] BeginPlay
  (Utilities|FlowControl|Sequence
    (:then_0
      (bind _asfort_player_pawn (Utilities|Casting|CastToFortPlayerPawn (Variables|Actor|GetInstigator))
        (:then
          (bind _output_get (Variables|Default|SetOwningPlayer _asfort_player_pawn))
          (bind _asb_paprika_ballistic_shield_weap (Utilities|Casting|CastToB_Paprika_BallisticShield_Weap (Variables|/Script/FortniteGame.FortPawn|GetCurrentWeapon _output_get))
            (:then
              (Variables|Default|SetOwningWeapon _asb_paprika_ballistic_shield_weap)
              (Transformation|AttachActorToComponent (Variables|/BallisticShieldGameplay/Gameplay/Actors/B_Paprika_BallisticShield_Weap.B_Paprika_BallisticShield_Weap_C|GetShieldMesh (Variables|Default|GetOwningWeapon)) "None" "SnapToTarget" "KeepRelative" "KeepWorld")
              (Variables|/BallisticShieldGameplay/Gameplay/Actors/B_Paprika_BallisticShield_Weap.B_Paprika_BallisticShield_Weap_C|SetBlockingBGA self _owningweapon)
              (Class|BPaprikaBallisticShieldWeap|OnBlockingBGAAdded _owningweapon))
            (:CastFailed
              _)))
        (:CastFailed
          _))
      ;;[1 0 0.05961] Doomsday button
      (Actor|DestroyActor))
    (:then_1
      (CallFunction|SetTeam (Team|GetActorTeam (Game|GetInstigator)))
      (Gameplay|ForceBuildingHealth 5e+04)
      (Transformation|SetRelativeScale3D (Variables|Default|GetBox) (select (Networking|IsServer) (* (Variables|/Script/Engine.SceneComponent|GetRelativeScale3D (Variables|Default|GetBox)) (Variables|Default|GetServerScaleMultiplier)) (Variables|/Script/Engine.SceneComponent|GetRelativeScale3D (Variables|Default|GetBox))))
      (CallFunction|ToggleBlockProjectiles true)
      (Utilities|IsValid _owningweapon_2
        (:Is%20Valid
          (Variables|Default|SetDamageCount (Variables|/BallisticShieldGameplay/Gameplay/Actors/B_Paprika_BallisticShield_Weap.B_Paprika_BallisticShield_Weap_C|GetInstancedStagger _owningweapon_2)))))
    (:then_2
      (Ability|GameplayEffect|AddLooseGameplayTags self (Variables|Default|GetLooseTagsToAddOnSpawn)))))

(event Custom|DebugToggleBallisticShieldServerCollision
  ;[1 0 0.05961] DEBUG
  (bind _output_get (Variables|Debug|SetbDebugShowServerCollision (not (Variables|Debug|GetbDebugShowServerCollision))))
  (CallFunction|SetTemporarilyTicking _output_get))

(event EventTick
  (params (DeltaSeconds :type Float%20%28single-precision%29))
  ;[1 0 0.05961] DEBUG
  (if (and (Editor|InEditororPIE) (Collision|IsCollisionEnabled (Variables|Default|GetBox)))
    (Rendering|Debug|DrawDebugBox (Transformation|GetWorldLocation (Variables|Default|GetBox)) (Components|Box|GetScaledBoxExtent (Variables|Default|GetBox)) (select (Networking|IsServer) "(R=0.000000,G=0.010310,B=1.000000,A=1.000000)" "(R=1.000000,G=0.011060,B=0.000000,A=1.000000)") (Transformation|GetWorldRotation (Variables|Default|GetBox)) 0.0 2.0)))

(event Custom|ToggleBlockProjectiles
  (params (Enabled? :type Boolean))
  ;[0.03429 0.3069 0.7899] Toggle bullet blocking effects
  (Collision|SetCollisionEnabled (Variables|Default|GetBox) (select Enabled? "QueryOnly" "NoCollision"))
  ;;[1 0 0.05961] Debug
  (if (and (Editor|InEditororPIE) (Variables|Default|GetDrawBoxCollision?))
    (Development|SetHiddeninGame (Variables|Default|GetBox) (not Enabled?) false)))

(event Custom|PostDamageTaken
  (params (DamageTaken :type Float%20%28double-precision%29))
  ;[0.03429 0.3069 0.7899] Play impacts and reset health after taking any damage
  (Ability|SendGameplayEventtoActor (Variables|Default|GetOwningPlayer) "(TagName=\"Event.Paprika.BallisticShield.Damage.Blocked\")" (Utilities|Struct|MakeGameplayEventData "(TagName=\"\")" 0 0 0 0 "()" "(GameplayTags=)" "(GameplayTags=)" DamageTaken))
  (GameplayCue|ExecuteGameplayCueOnActor(Burst) self (Variables|Default|GetGC_ShieldImpacted) (Variables|Default|GetDamageParams))
  (Gameplay|ForceBuildingHealth 5e+04))

(event EventOnDamageServer
  (params (Damage :type Float%20%28single-precision%29) (DamageTags :type Gameplay%20Tag%20Container%20Structure%20%28by%20ref%29) (Momentum :type Vector :default "0, 0, 0") (HitInfo :type Hit%20Result%20Structure%20%28by%20ref%29) (InstigatedBy :type Controller%20Object%20Reference) (DamageCauser :type Actor%20Object%20Reference) (EffectContext :type Gameplay%20Effect%20Context%20Handle%20Structure))
  (bind _self self)
  (bind _owningplayer (Variables|Default|GetOwningPlayer))
  (bind _damageparams (Variables|Default|GetDamageParams))
  ;[0.03429 0.3069 0.7899] Damage tracking + visuals
  (Variables|Default|SetDamageParams (CallFunction|GetDamageAbsorbParams HitInfo DamageCauser DamageTags))
  (if (Default|CheckCanBuildStagger?)
    (bind _output_get (Variables|Default|SetDamageCount (+ (CallFunction|CheckDamageMultipliers Damage DamageTags) (Variables|Default|GetDamageCount))))
    (if (>= _output_get (ScalableFloat|GetValueatLevel (Variables|Default|GetHF_BlockHealth)))
      (Variables|Default|SetInStaggerState? true)
      (GameplayCue|ExecuteGameplayCueOnActor(Burst) _self (Variables|Default|GetGC_BallisticShieldBreak) _damageparams)
      (else
        (GameplayCue|ExecuteGameplayCueOnActor(Burst) _self (Variables|Default|GetGC_DamageAbsorbed) _damageparams)
        (Ability|SendGameplayEventtoActor _owningplayer (Variables|Default|GetEvent_StaggerAccrued))
        (GameplayCue|ExecuteGameplayCueOnActor(Burst) _owningplayer (Variables|Default|GetGC_StaggerBuilt) (Ability|GameplayCue|MakeGameplayCueParameters (CallFunction|GetDamageCountasPercentage)))))
    (CallFunction|ResetDecay)
    (CallFunction|PostDamageTaken Damage)
    (else
      (CallFunction|PostDamageTaken Damage))))

(event Custom|ResetDecay
  ;[0.2248 0.7344 0.0859] Handle damage buildup decay when user has not taken damage for a duration
  (CallFunction|CheckandClearTimers)
  (bind _returnvalue (Utilities|Time|SetTimerbyEvent (delegate Custom|BeginDamageDecay) (ScalableFloat|GetValueatLevel (Variables|Decay|GetHF_DecayStartTime))))
  (Variables|Decay|SetDecayStartTimer _returnvalue))

(event Custom|BeginDamageDecay
  ;[0.2248 0.7344 0.0859] Handle damage buildup decay when user has not taken damage for a duration
  (bind _returnvalue (Utilities|Time|SetTimerbyEvent (delegate Custom|DecayDamage) (ScalableFloat|GetValueatLevel (Variables|Decay|GetHF_DecayRate)) true))
  (Variables|Decay|SetDecayTimer _returnvalue))

(event Custom|DecayDamage
  (bind _hf_damagedecaypertick (Variables|Decay|GetHF_DamageDecayPerTick))
  (bind _returnvalue (ScalableFloat|GetValueatLevel _hf_damagedecaypertick))
  (bind _damagecount (Variables|Default|GetDamageCount))
  (bind _returnvalue_2 (- _damagecount _returnvalue))
  ;[0.2248 0.7344 0.0859] Handle damage buildup decay when user has not taken damage for a duration
  (if (<= _returnvalue_2 0)
    (Variables|Default|SetDamageCount 0.0)
    (CallFunction|CheckandClearTimers)
    (else
      (Variables|Default|SetDamageCount _returnvalue_2))))

(fn CheckAndClearTimers
  (bind _decaystarttimer (Variables|Decay|GetDecayStartTimer))
  (bind _decaytimer (Variables|Decay|GetDecayTimer))
  ;[0.778 0.04391 1] Wipe timers when damage is taken
  (Utilities|FlowControl|Sequence
    (:then_0
      (if (Utilities|Time|IsTimerActivebyHandle _decaystarttimer)
        (Utilities|Time|ClearandInvalidateTimerbyHandle _decaystarttimer)))
    (:then_1
      (if (Utilities|Time|IsTimerActivebyHandle _decaytimer)
        (Utilities|Time|ClearandInvalidateTimerbyHandle _decaytimer)))
    (:then_2
      (return))))

(fn CheckCanBuildStagger?
  ;[0.778 0.04391 1] Check player state and stagger state
  (return (and (not (Variables|Default|GetInStaggerState?)) (Utilities|Array|ContainsItem (Variables|Default|GetDamagableActionStates) (Variables|/BallisticShieldGameplay/Gameplay/Actors/B_Paprika_BallisticShield_Weap.B_Paprika_BallisticShield_Weap_C|GetShieldActionState (Variables|Default|GetOwningWeapon))))))

(fn CheckDamageMultipliers
  (params (Damage :type Float%20%28double-precision%29) (Tags :type Gameplay%20Tag%20Container%20Structure))
  ;[0.03429 0.3069 0.7899] Iterate our damage multiplier data and see if this instance of damage should be multiplied
  (bind _array_element (Utilities|Array|ForEachLoopwithBreak (Variables|Default|GetDamageMultiplers))
    (:LoopBody
      (bind (_array_element_damagetags_3_788373e142b2fe9b40b10f82cada4b16 _array_element_multiplier_7_44facfda418066749597d298806026bd) (Utilities|Struct|BreakFortBShieldMultiplierData _array_element))
      (bind _output_get (LocalVariables|SetLocalBool (GameplayTags|HasAnyTags Tags _array_element_damagetags_3_788373e142b2fe9b40b10f82cada4b16 true)))
      (if _output_get
        (LocalVariables|SetLocalFloat (* Damage (ScalableFloat|GetValueatLevel _array_element_multiplier_7_44facfda418066749597d298806026bd)))))
    (:Completed
      _))
  (return (Math|Float|SelectFloat (LocalVariables|GetLocalFloat) Damage (LocalVariables|GetLocalBool))))

(fn EndStagger
  (Variables|Default|SetDamageCount 0.0)
  (Variables|Default|SetInStaggerState? false))

(fn GetDamageAbsorbParams
  (params (Hit_bBlockingHit :type Boolean) (Hit_bInitialOverlap :type Boolean) (Hit_Time :type Float%20%28single-precision%29) (Hit_Distance :type Float%20%28single-precision%29) (Hit_Location :type Vector :default "0, 0, 0") (Hit_ImpactPoint :type Vector :default "0, 0, 0") (Hit_Normal :type Vector :default "0, 0, 0") (Hit_ImpactNormal :type Vector :default "0, 0, 0") (Hit_PhysMat :type Physical%20Material%20Object%20Reference) (Hit_HitActor :type Actor%20Object%20Reference) (Hit_HitComponent :type Primitive%20Component%20Object%20Reference) (Hit_HitBoneName :type Name :default "None") (Hit_BoneName :type Name :default "None") (Hit_HitItem :type Integer) (Hit_ElementIndex :type Integer) (Hit_FaceIndex :type Integer) (Hit_TraceStart :type Vector :default "0, 0, 0") (Hit_TraceEnd :type Vector :default "0, 0, 0") (DamageCauser :type Actor%20Object%20Reference) (DamageTags :type Gameplay%20Tag%20Container%20Structure))
  (return (Ability|GameplayCue|MakeGameplayCueParameters :NormalizedMagnitude (CallFunction|GetDamageCountasPercentage) :RawMagnitude 0.0 :AggregatedSourceTags DamageTags :Location Hit_Location :Normal Hit_ImpactNormal :EffectCauser DamageCauser :SourceObject (Variables|Default|GetOwningPlayer) :GameplayEffectLevel 1 :AbilityLevel 1 :bReplicateLocationWhenUsingMinimalRepProxy false)))

(fn GetDamageCountAsPercentage
  (return (Math|Float|Clamp(Float) (Math|Float|SafeDivide (Variables|Default|GetDamageCount) (ScalableFloat|GetValueatLevel (Variables|Default|GetHF_BlockHealth))))))

(fn OnRep_InStaggerState?
  (if (Variables|Default|GetInStaggerState?)
    (Ability|SendGameplayEventtoActor (Variables|Default|GetOwningPlayer) "(TagName=\"Event.Paprika.BallisticShield.Player.Stagger\")")
    (CallFunction|ToggleBlockProjectiles false)
    (else
      (Gameplay|ForceBuildingHealth 1e+04)
      (CallFunction|ToggleBlockProjectiles true)
      (Development|PrintString "block reenabled"))))

(fn OnRep_ResetDecay?
  (if (Variables|Decay|GetResetDecay?)
    (CallFunction|ResetDecay)
    (else
      (CallFunction|CheckandClearTimers)))
  (return))

(fn ConstructionScript)
