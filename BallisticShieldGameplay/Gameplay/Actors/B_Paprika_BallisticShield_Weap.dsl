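; Ballistic shield weapon that attaches a shield to its owner and tracks idle, blocking and charging states.
; Targeting and charge gameplay events set the state, which refreshes looping gameplay cues.
; The server attaches the shield to root rather than the left hand, using a timeline to move it between idle and blocking poses.
; It also registers a sprint extension, toggles the block volume's projectile blocking with visibility, and saves stagger on unequip.
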(event FortWeapon|EventOnEquippedAndAttached
  (bind _weaponowner (Variables|Default|GetWeaponOwner))
  (Utilities|FlowControl|Sequence
    (:then_0
      ;[0.03429 0.3069 0.7899] Attach shield to left hand (or root in case of server)
      (bind _asfort_pawn (Utilities|Casting|CastToFortPawn (Game|GetInstigator)))
      (Variables|Default|SetWeaponOwner _asfort_pawn)
      (Transformation|AttachComponentToComponent (Variables|Default|GetShieldMesh) (Attachment|GetSceneComponenttoAttachWeaponTo _asfort_pawn) (select (Networking|IsServer) "root" "hand_attach_l") "KeepRelative" "KeepRelative" "KeepRelative" true)
      ;[0.03429 0.3069 0.7899] Adjust server shield collision
      (if (Networking|IsServer)
        (CallFunction|UpdateServerShieldLocationandRotation false)
        (else
          _))
      (CallFunction|SetShieldActionState "Idling")
      ;[0.03429 0.3069 0.7899] Listen to charge event and set player state
      (Ability|Async|EndAction (Variables|Default|GetAsyncAction_EventChargeStarted))
      (bind _asynctaskproxy (Ability|Async|WaitGameplayEventToActor _weaponowner (Variables|Default|GetEvent_ChargeStarted))
        (:then
          (Variables|Default|SetAsyncAction_EventChargeStarted _asynctaskproxy)
          (Ability|Async|EndAction (Variables|Default|GetAsyncAction_EventChargeEnded))
          (bind _asynctaskproxy_2 (Ability|Async|WaitGameplayEventToActor _weaponowner (Variables|Default|GetEvent_ChargeEnded))
            (:then
              (Variables|Default|SetAsyncAction_EventChargeEnded _asynctaskproxy_2))
            (:EventReceived
              (if (not (FortWeapon|IsTargeting))
                (CallFunction|SetShieldActionState "Idling")))))
        (:EventReceived
          (CallFunction|SetShieldActionState "Charging"))))
    (:then_1
      (Rendering|SetVisibility (Variables|Default|GetShieldComponent) true false)
      (Variables|Default|SetWeaponOwnerController (Pawn|GetController (Variables|Default|GetWeaponOwner)))
      (Sprint|AddCustomSprintMMEClass (Actor|GetComponentbyClass (Variables|Default|GetWeaponOwnerController) "/TacticalSprintGame/Gameplay/TacticalSprintControllerComponent.TacticalSprintControllerComponent_C") (Variables|Default|GetCustomTacSprintExtensionClass))
      (CallFunction|SetShieldActionState "Idling"))))

(event Custom|Grass%20Push
  (Game|SpawnActorfromClass "/Game/Effects/Niagara/Blueprints/PhsyicsFields/FN_RadialForce.FN_RadialForce_C" (Transformation|GetActorTransform (Variables|Default|GetWeaponOwner)) "Undefined" "MultiplyWithRoot" 0 0.2 1.0 2.5e+02 0.2 0.0 0))

(event FortWeapon|EventOnUnEquippedAndDetached
  (bind _blockingbga (Variables|Default|GetBlockingBGA))
  (bind _owningplayer (Variables|/BallisticShieldGameplay/Gameplay/Actors/BGA_Paprika_BallisticShield_BlockVolume.BGA_Paprika_BallisticShield_BlockVolume_C|GetOwningPlayer _blockingbga))
  (Utilities|FlowControl|Sequence
    (:then_0
      (Rendering|SetVisibility (Variables|Default|GetShieldComponent) false false)
      (Networking|SwitchHasAuthority
        (:Authority
          (CallFunction|SaveInstancedStagger))
        (:Remote
          _))
      (Sprint|RemoveCustomSprintMMEClass (Actor|GetComponentbyClass (Variables|Default|GetWeaponOwnerController) "/TacticalSprintGame/Gameplay/TacticalSprintControllerComponent.TacticalSprintControllerComponent_C") (Variables|Default|GetCustomTacSprintExtensionClass))
      (CallFunction|SetShieldActionState "Idling"))
    (:then_1
      (Ability|Async|EndAction (Variables|Default|GetAsyncAction_EventChargeStarted))
      (Ability|Async|EndAction (Variables|Default|GetAsyncAction_EventChargeEnded))
      (Utilities|IsValid _blockingbga
        (:Is%20Valid
          (Utilities|IsValid _owningplayer
            (:Is%20Valid
              (Class|FortPawn|RemoveWeaponTraceIgnoreActor _owningplayer _blockingbga))))))))

(event EventOnSetTargeting
  (params (bNewIsTargeting :type Boolean))
  ;[0.03429 0.3069 0.7899] Set player state on aim down sights
  (|CallParentFunction bNewIsTargeting)
  (CallFunction|SetShieldActionState (select bNewIsTargeting "Blocking" "Idling")))

(event Custom|UpdateServerShieldLocationAndRotation
  (params (bIsInBlockingState :type Boolean))
  (bind (_direction _movetoblock _movetoidle) (Timeline|Timeline_MoveServerShield PlayFromStart)
    (:Update
      (Transformation|SetRelativeLocationAndRotation (Variables|Default|GetShieldMesh) (Math|Vector|Lerp(Vector) (Variables|Default|GetServerShieldLocation_Idle) (Variables|Default|GetServerShieldLocation_Blocking) (select bIsInBlockingState _movetoblock (- 1.0 _movetoidle))) (Math|Rotator|Lerp(Rotator) (Variables|Default|GetServerShieldRotation_Idle) (Variables|Default|GetServerShieldRotation_Blocking) (select bIsInBlockingState _movetoblock (- 1.0 _movetoidle))) false false))
    (:Finished
      (Transformation|SetRelativeLocationAndRotation (Variables|Default|GetShieldMesh) (select bIsInBlockingState (Variables|Default|GetServerShieldLocation_Blocking) (Variables|Default|GetServerShieldLocation_Idle)) (select bIsInBlockingState (Variables|Default|GetServerShieldRotation_Blocking) (Variables|Default|GetServerShieldRotation_Idle)) false false))))

(event EventOnWeaponVisibilityChanged
  (params (bVisible :type Boolean) (bSetForLocalControllerOnly :type Boolean))
  (if (not bSetForLocalControllerOnly)
    (Class|BGAPaprikaBallisticShieldBlockVolume|ToggleBlockProjectiles (Variables|Default|GetBlockingBGA) bVisible)))

(event Custom|OnBlockingBGAAdded
  (bind _blockingbga (Variables|Default|GetBlockingBGA))
  (bind _owningplayer (Variables|/BallisticShieldGameplay/Gameplay/Actors/BGA_Paprika_BallisticShield_BlockVolume.BGA_Paprika_BallisticShield_BlockVolume_C|GetOwningPlayer _blockingbga))
  (Utilities|IsValid _blockingbga
    (:Is%20Valid
      (Utilities|IsValid _owningplayer
        (:Is%20Valid
          (Class|FortPawn|AddWeaponTraceIgnoreActor _owningplayer _blockingbga))))))

(fn GetBallisticShieldData
  (bind _staggerbuilduppercent (CallFunction|GetStaggerBuildupasPercentage))
  (return (Variables|Default|GetShieldActionState) _staggerbuilduppercent))

(fn GetStaggerBuildupAsPercentage
  (Utilities|IsValid (Variables|Default|GetBlockingBGA)
    (:Is%20Valid
      (return (Class|BGAPaprikaBallisticShieldBlockVolume|GetDamageCountasPercentage (Variables|Default|GetBlockingBGA))))
    (:Is%20Not%20Valid
      (return 0.0))))

(fn SaveInstancedStagger
  (bind _blockingbga (Variables|Default|GetBlockingBGA))
  ;[0.4546 0.06953 0.6667] Save our stagger value before switching away from the weapon
  (Utilities|IsValid _blockingbga
    (:Is%20Valid
      (bind _output_get (Variables|Default|SetInstancedStagger (Math|Float|Clamp(Float) (Variables|/BallisticShieldGameplay/Gameplay/Actors/BGA_Paprika_BallisticShield_BlockVolume.BGA_Paprika_BallisticShield_BlockVolume_C|GetDamageCount _blockingbga) 0.0 (ScalableFloat|GetValueatLevel (Variables|/BallisticShieldGameplay/Gameplay/Actors/BGA_Paprika_BallisticShield_BlockVolume.BGA_Paprika_BallisticShield_BlockVolume_C|GetHF_BlockHealth _blockingbga)))))
      (Development|PrintString (Utilities|String|Append "Stagger saved: " (Utilities|String|ToString(Float) _output_get)))
      (return))))

(fn SetShieldActionState
  (params (NewState :type EBallisticShieldPlayerActionState%20Enum :default "Idling"))
  (bind _weaponowner (Variables|Default|GetWeaponOwner))
  (bind _tempparams (LocalVariables|GetTempParams))
  ;[0.4895 0.1317 1] Update shield state, remove and apply looping cues based on new state
  (bind _output_get (Variables|Default|SetShieldActionState NewState))
  (GameplayCue|RemoveGameplayCueOnActor(Looping) _weaponowner (Variables|Default|GetGC_Blocking) _tempparams)
  (GameplayCue|RemoveGameplayCueOnActor(Looping) _weaponowner (Variables|Default|GetGC_Charging) _tempparams)
  (switch Utilities|FlowControl|Switch|SwitchonEBallisticShieldPlayerActionState _output_get
    (:Idling
      _)
    (:Blocking
      (GameplayCue|AddGameplayCueOnActor(Looping) (Variables|Default|GetWeaponOwner) (Utilities|Select :Index (Variables|Default|GetShieldActionState) :Blocking (Variables|Default|GetGC_Blocking) :Charging (Variables|Default|GetGC_Charging)) (LocalVariables|GetTempParams))))
  (if (Networking|IsServer)
    (CallFunction|UpdateServerShieldLocationandRotation (Utilities|Select :Index (Variables|Default|GetShieldActionState) :Idling false :Blocking true :Charging true)))
  (return))

(fn ConstructionScript)
